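# Niagara system that ejects a burst of weapon shell casings as mesh particles, driven almost entirely by User parameters.
# One emitter spawns an instant burst by scale and probability inside an X-axis cylinder, with velocity, gravity and inherited parent velocity.
# User Shell Type sets VisibilityTag for the five mesh renderers.
# Casings spin at randomized rates, shrink late via curve, and receive ShellColor's RGB through dynamic material parameters.
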NiagaraSystem /Game/Weapons/FORT_Generic_Effects/ShellEject/NS_ShellsFire_Burst
dependencies {
 usedDataInterfaces [&NiagaraDataInterfaceCurve]
 usedDynamicInputs []
 usedModules [
  &SystemState
  &self:SystemSpawnScript.SystemScriptSource.SystemScriptGraph.NiagaraNodeAssignment_0.SetVariables_#1
 ]
}
emitters {Shell_Eject_Burst {
  emitterUpdateScript [
   &EmitterState()
   &SpawnBurst_Instantaneous(
    Spawn Count=&MakeIntFromFloat(Float=$User.BurstScale)
    Spawn Probability=$User.SpawnProbability
    Use Spawn Probability=true
   )
  ]
  particleSpawnScript [
   /Niagara/Modules/Spawn/Initialization/InitializeParticle(
    Module.Write Color=true
    Module.Write Lifetime=true
    Module.Write Mass=true
    Module.Write Material Random=true
    Module.Write Position=false
    Module.Write Scale=true
    Module.Write SpriteRotation=false
    Module.Write SpriteSize=false
    Sprite Size=<5 5>
   )
   /SetVariables(Particles.VisibilityTag=$User.Shell Type) {isSetParameters=true}
   /SetVariables(Particles.UniformMeshOrientation_Initial:Vector3f=&RandomRangeVector(Minimum=<0.2 0.2 0.2>, Maximum=<1 1 1>)) {isSetParameters=true}
   &CylinderLocation(
    Cylinder Height:NiagaraFloat=$User.CylinderHeight
    Cylinder Origin:Vector3f=/Niagara/DynamicInputs/Helpers/SimulationPosition()
    Cylinder Radius=$User.CylinderRadius
    MaskSpawnGroup=false
    Module.Enable Lathe Profile=false
    Module.Override Randomness=false
    Module.Surface Only=false
    Orientation Axis:ENiagaraOrientationAxis=::XAxis
   )
   &AddVelocity(Velocity=$User.Velocity)
  ]
  particleUpdateScript [
   &ParticleState()
   &AccelerationForce(Acceleration=$User.Gravity)
   &InheritVelocity(Module.Limit Speed=true, Velocity Limit=10, Velocity Scale=$User.InheritParentVelocity)
   &SolveForcesAndVelocity(Clamp Velocity=false, Limit Acceleration=false)
   &ScaleMeshSize(Scale Factor=$User.Size)
   &ScaleMeshSize(Scale Factor=&VectorFromFloat(Value=&FloatFromCurve(FloatCurve=DataInterface() {propertyValues {Curve {keys [<at 0.7517 1 RCIM_Cubic>, {interpMode=RCIM_Cubic, tangentMode=RCTM_User, time=0.8642, value=1}, <at 1 0 RCIM_Cubic>]}}}))) {module=ScaleMeshSize001}
   &ScaleMeshSize(Scale Factor=&VectorFromFloat(Value=&FloatFromCurve(FloatCurve=DataInterface() {propertyValues {Curve {keys [<at 0.7517 1 RCIM_Cubic>, {interpMode=RCIM_Cubic, tangentMode=RCTM_User, time=0.8642, value=1.3298}, <at 1 0 RCIM_Cubic>]}}}))) {enabled=false, module=ScaleMeshSize002}
   /Niagara/Modules/Update/Orientation/V2/UpdateMeshOrientation(Rotation Vector:Vector3f=&Multiply_Vector(A=$Particles.UniformMeshOrientation_Initial, B=$User.ShellsRotationRate))
   &DynamicMaterialParameters(
    Index 0 Param 1=&MakeFloatFromLinearColor(LinearColor=$User.ShellColor, Channel:ENiagaraLinearColor_Channels=::NewEnumerator0)
    Index 0 Param 2=&MakeFloatFromLinearColor(LinearColor=$User.ShellColor, Channel:ENiagaraLinearColor_Channels=::NewEnumerator1)
    Index 0 Param 3=&MakeFloatFromLinearColor(LinearColor=$User.ShellColor, Channel:ENiagaraLinearColor_Channels=::NewEnumerator2)
    Module.Param0WriteEnabled=true
    Module.Param1WriteEnabled=true
    Module.Param2WriteEnabled=true
    Module.Param3WriteEnabled=false
   )
   /SetVariables(Particles.DEBUG_ShellType:NiagaraInt32=$User.Shell Type) {isSetParameters=true}
  ]
  rendererClasses [&NiagaraMeshRendererProperties]
  renderers [
   {rendererClass=&NiagaraMeshRendererProperties, rendererIndex=0}
   {rendererClass=&NiagaraMeshRendererProperties, rendererIndex=1}
   {rendererClass=&NiagaraMeshRendererProperties, rendererIndex=2}
   {rendererClass=&NiagaraMeshRendererProperties, rendererIndex=3}
   {rendererClass=&NiagaraMeshRendererProperties, rendererIndex=4}
  ]
 }}
summary {
 systemName=NS_ShellsFire_Burst
 emitters [{emitterName=Shell_Eject_Burst, rendererClasses [&NiagaraMeshRendererProperties]}]
 userVariables [
  User.BurstScale {type=NiagaraFloat, defaultValue=1, description=""}
  User.CylinderHeight {type=NiagaraFloat, defaultValue=1, description=""}
  User.CylinderRadius {type=NiagaraFloat, defaultValue=0, description=""}
  User.Gravity {type=Vector3f, description="", defaultValue {struct=Vector3f, value=<0 0 -1500>}}
  User.InheritParentVelocity {type=Vector3f, description="", defaultValue {struct=Vector3f, value=<0.2 0.2 0.2>}}
  User.Shell Type {type=NiagaraInt32}
  User.ShellColor {type=LinearColor, description="", defaultValue {struct=LinearColor, value=<rgba 0 0 0 1>}}
  User.ShellsRotationRate {type=Vector3f, description="", defaultValue {struct=Vector3f, value=<1 1 1>}}
  User.Size {type=Vector3f, description="", defaultValue {struct=Vector3f, value=<1 1 1>}}
  User.SpawnProbability {type=NiagaraFloat, defaultValue=1, description=""}
  User.Velocity {type=Vector3f, description="", defaultValue {struct=Vector3f, value=<0 0 0>}}
 ]
}
