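# Sound cue for ejected weapon shell impacts, choosing pistol, rifle, sniper or shotgun sounds by the ShellType parameter.
# A quality-level node selects either large random wave pools or, for the other three levels, two-wave pools per type.
# Both paths vary volume slightly and add a roughly half-second random delay when bEnableDelay is true.
# Only the large sniper pool is pitched down, and a concurrency override limits playback to two instances.
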SoundCue /Game/Sounds/Fort_Weapons/Shells/Weapon_Shells_Cue
root=QualityLevel_2
nodes [
 {id=Branch_0, children [Delay_0,Switch_0,Switch_0], properties {boolParameterName=bEnableDelay}}
 {id=Branch_1, children [Delay_1,Switch_1,Switch_1], properties {boolParameterName=bEnableDelay}}
 {id=Delay_0, children [Switch_0], properties {delayMax=0.55, delayMin=0.45}}
 {id=Delay_1, children [Switch_1], properties {delayMax=0.55, delayMin=0.45}}
 {id=Mixer_0, children [Random_2], properties {inputVolume [1.6]}}
 {id=Mixer_1, children [Modulator_2], properties {inputVolume [0.65]}}
 {id=Mixer_2, children [Random_3], properties {inputVolume [1]}}
 {id=Mixer_3, children [Random_4], properties {inputVolume [1]}}
 {id=Mixer_4, children [Random_10], properties {inputVolume [1.6]}}
 {id=Mixer_5, children [Random_11], properties {inputVolume [1.2]}}
 {id=Mixer_6, children [Random_12], properties {inputVolume [1]}}
 {id=Mixer_7, children [Random_13], properties {inputVolume [1]}}
 {id=Modulator_0, children [Branch_0], properties {volumeMax=1.1, volumeMin=0.9}}
 {id=Modulator_1, children [Branch_1], properties {volumeMax=1.1, volumeMin=0.9}}
 {id=Modulator_2, children [Random_5], properties {pitchMax=0.8, pitchMin=0.8, volumeMax=1, volumeMin=1}}
 {id=QualityLevel_2, children [Modulator_0,Modulator_1,Modulator_1,Modulator_1]}
 {id=Random_10, children [WavePlayer_84, WavePlayer_85], properties {weights [1,1]}}
 {id=Random_11, children [WavePlayer_136, WavePlayer_137], properties {weights [1,1]}}
 {id=Random_12, children [WavePlayer_113, WavePlayer_114], properties {weights [1,1]}}
 {id=Random_13, children [WavePlayer_134, WavePlayer_135], properties {weights [1,1]}}
 {
  id=Random_2
  children [
   WavePlayer_14
   WavePlayer_15
   WavePlayer_16
   WavePlayer_17
   WavePlayer_18
   WavePlayer_19
   WavePlayer_20
   WavePlayer_21
   WavePlayer_22
   WavePlayer_23
   WavePlayer_24
   WavePlayer_25
   WavePlayer_26
  ]
  properties {weights [1,1,1,1,1,1,1,1,1,1,1,1,1]}
 }
 {
  id=Random_3
  children [
   WavePlayer_27
   WavePlayer_28
   WavePlayer_29
   WavePlayer_30
   WavePlayer_31
   WavePlayer_32
   WavePlayer_33
   WavePlayer_34
   WavePlayer_35
   WavePlayer_36
   WavePlayer_37
   WavePlayer_38
   WavePlayer_39
   WavePlayer_40
   WavePlayer_41
   WavePlayer_42
   WavePlayer_43
   WavePlayer_44
  ]
  properties {weights [1,1,1,1,1,1,1,1,1,1,1,1,1,1,1,1,1,1]}
 }
 {
  id=Random_4
  children [
   WavePlayer_45
   WavePlayer_46
   WavePlayer_47
   WavePlayer_48
   WavePlayer_49
   WavePlayer_50
   WavePlayer_51
   WavePlayer_52
   WavePlayer_53
   WavePlayer_54
   WavePlayer_55
   WavePlayer_56
   WavePlayer_57
   WavePlayer_58
   WavePlayer_59
   WavePlayer_60
   WavePlayer_61
   WavePlayer_62
   WavePlayer_63
   WavePlayer_64
   WavePlayer_65
  ]
  properties {weights [1,1,1,1,1,1,1,1,1,1,1,1,1,1,1,1,1,1,1,1,1]}
 }
 {
  id=Random_5
  children [
   WavePlayer_66
   WavePlayer_67
   WavePlayer_68
   WavePlayer_69
   WavePlayer_70
   WavePlayer_71
   WavePlayer_72
   WavePlayer_73
   WavePlayer_74
   WavePlayer_75
   WavePlayer_76
   WavePlayer_77
   WavePlayer_78
   WavePlayer_79
   WavePlayer_80
   WavePlayer_81
   WavePlayer_82
   WavePlayer_83
  ]
  properties {weights [1,1,1,1,1,1,1,1,1,1,1,1,1,1,1,1,1,1]}
 }
 {id=Switch_0, children [null,Mixer_3,Mixer_2,Mixer_1,null,Mixer_0], properties {intParameterName=ShellType}}
 {id=Switch_1, children [null,Mixer_7,Mixer_6,Mixer_5,null,Mixer_4], properties {intParameterName=ShellType}}
 {id=WavePlayer_113, properties {soundWaveAssetPtr=/Game/Sounds/Fort_Weapons/AmmoImpacts/Ammo_Impact_Rifles_02}}
 {id=WavePlayer_114, properties {soundWaveAssetPtr=/Game/Sounds/Fort_Weapons/AmmoImpacts/Ammo_Impact_Rifles_01}}
 {id=WavePlayer_134, properties {soundWaveAssetPtr=/Game/Sounds/Fort_Weapons/AmmoImpacts/Ammo_Impact_Pistols_02}}
 {id=WavePlayer_135, properties {soundWaveAssetPtr=/Game/Sounds/Fort_Weapons/AmmoImpacts/Ammo_Impact_Pistols_01}}
 {id=WavePlayer_136, properties {soundWaveAssetPtr=/Game/Sounds/Fort_Weapons/AmmoImpacts/Ammo_Impact_Sniper_01}}
 {id=WavePlayer_137, properties {soundWaveAssetPtr=/Game/Sounds/Fort_Weapons/AmmoImpacts/Ammo_Impact_Sniper_02}}
 {id=WavePlayer_14, properties {soundWaveAssetPtr=/Game/Sounds/Fort_Weapons/AmmoImpacts/Ammo_Impact_Shotgun_13}}
 {id=WavePlayer_15, properties {soundWaveAssetPtr=/Game/Sounds/Fort_Weapons/AmmoImpacts/Ammo_Impact_Shotgun_12}}
 {id=WavePlayer_16, properties {soundWaveAssetPtr=/Game/Sounds/Fort_Weapons/AmmoImpacts/Ammo_Impact_Shotgun_11}}
 {id=WavePlayer_17, properties {soundWaveAssetPtr=/Game/Sounds/Fort_Weapons/AmmoImpacts/Ammo_Impact_Shotgun_10}}
 {id=WavePlayer_18, properties {soundWaveAssetPtr=/Game/Sounds/Fort_Weapons/AmmoImpacts/Ammo_Impact_Shotgun_09}}
 {id=WavePlayer_19, properties {soundWaveAssetPtr=/Game/Sounds/Fort_Weapons/AmmoImpacts/Ammo_Impact_Shotgun_08}}
 {id=WavePlayer_20, properties {soundWaveAssetPtr=/Game/Sounds/Fort_Weapons/AmmoImpacts/Ammo_Impact_Shotgun_07}}
 {id=WavePlayer_21, properties {soundWaveAssetPtr=/Game/Sounds/Fort_Weapons/AmmoImpacts/Ammo_Impact_Shotgun_06}}
 {id=WavePlayer_22, properties {soundWaveAssetPtr=/Game/Sounds/Fort_Weapons/AmmoImpacts/Ammo_Impact_Shotgun_05}}
 {id=WavePlayer_23, properties {soundWaveAssetPtr=/Game/Sounds/Fort_Weapons/AmmoImpacts/Ammo_Impact_Shotgun_04}}
 {id=WavePlayer_24, properties {soundWaveAssetPtr=/Game/Sounds/Fort_Weapons/AmmoImpacts/Ammo_Impact_Shotgun_03}}
 {id=WavePlayer_25, properties {soundWaveAssetPtr=/Game/Sounds/Fort_Weapons/AmmoImpacts/Ammo_Impact_Shotgun_02}}
 {id=WavePlayer_26, properties {soundWaveAssetPtr=/Game/Sounds/Fort_Weapons/AmmoImpacts/Ammo_Impact_Shotgun_01}}
 {id=WavePlayer_27, properties {soundWaveAssetPtr=/Game/Sounds/Fort_Weapons/AmmoImpacts/Ammo_Impact_Rifles_18}}
 {id=WavePlayer_28, properties {soundWaveAssetPtr=/Game/Sounds/Fort_Weapons/AmmoImpacts/Ammo_Impact_Rifles_17}}
 {id=WavePlayer_29, properties {soundWaveAssetPtr=/Game/Sounds/Fort_Weapons/AmmoImpacts/Ammo_Impact_Rifles_16}}
 {id=WavePlayer_30, properties {soundWaveAssetPtr=/Game/Sounds/Fort_Weapons/AmmoImpacts/Ammo_Impact_Rifles_15}}
 {id=WavePlayer_31, properties {soundWaveAssetPtr=/Game/Sounds/Fort_Weapons/AmmoImpacts/Ammo_Impact_Rifles_14}}
 {id=WavePlayer_32, properties {soundWaveAssetPtr=/Game/Sounds/Fort_Weapons/AmmoImpacts/Ammo_Impact_Rifles_13}}
 {id=WavePlayer_33, properties {soundWaveAssetPtr=/Game/Sounds/Fort_Weapons/AmmoImpacts/Ammo_Impact_Rifles_12}}
 {id=WavePlayer_34, properties {soundWaveAssetPtr=/Game/Sounds/Fort_Weapons/AmmoImpacts/Ammo_Impact_Rifles_11}}
 {id=WavePlayer_35, properties {soundWaveAssetPtr=/Game/Sounds/Fort_Weapons/AmmoImpacts/Ammo_Impact_Rifles_10}}
 {id=WavePlayer_36, properties {soundWaveAssetPtr=/Game/Sounds/Fort_Weapons/AmmoImpacts/Ammo_Impact_Rifles_09}}
 {id=WavePlayer_37, properties {soundWaveAssetPtr=/Game/Sounds/Fort_Weapons/AmmoImpacts/Ammo_Impact_Rifles_08}}
 {id=WavePlayer_38, properties {soundWaveAssetPtr=/Game/Sounds/Fort_Weapons/AmmoImpacts/Ammo_Impact_Rifles_07}}
 {id=WavePlayer_39, properties {soundWaveAssetPtr=/Game/Sounds/Fort_Weapons/AmmoImpacts/Ammo_Impact_Rifles_06}}
 {id=WavePlayer_40, properties {soundWaveAssetPtr=/Game/Sounds/Fort_Weapons/AmmoImpacts/Ammo_Impact_Rifles_05}}
 {id=WavePlayer_41, properties {soundWaveAssetPtr=/Game/Sounds/Fort_Weapons/AmmoImpacts/Ammo_Impact_Rifles_04}}
 {id=WavePlayer_42, properties {soundWaveAssetPtr=/Game/Sounds/Fort_Weapons/AmmoImpacts/Ammo_Impact_Rifles_03}}
 {id=WavePlayer_43, properties {soundWaveAssetPtr=&Ammo_Impact_Rifles_02}}
 {id=WavePlayer_44, properties {soundWaveAssetPtr=&Ammo_Impact_Rifles_01}}
 {id=WavePlayer_45, properties {soundWaveAssetPtr=/Game/Sounds/Fort_Weapons/AmmoImpacts/Ammo_Impact_Pistols_21}}
 {id=WavePlayer_46, properties {soundWaveAssetPtr=/Game/Sounds/Fort_Weapons/AmmoImpacts/Ammo_Impact_Pistols_20}}
 {id=WavePlayer_47, properties {soundWaveAssetPtr=/Game/Sounds/Fort_Weapons/AmmoImpacts/Ammo_Impact_Pistols_19}}
 {id=WavePlayer_48, properties {soundWaveAssetPtr=/Game/Sounds/Fort_Weapons/AmmoImpacts/Ammo_Impact_Pistols_18}}
 {id=WavePlayer_49, properties {soundWaveAssetPtr=/Game/Sounds/Fort_Weapons/AmmoImpacts/Ammo_Impact_Pistols_17}}
 {id=WavePlayer_50, properties {soundWaveAssetPtr=/Game/Sounds/Fort_Weapons/AmmoImpacts/Ammo_Impact_Pistols_16}}
 {id=WavePlayer_51, properties {soundWaveAssetPtr=/Game/Sounds/Fort_Weapons/AmmoImpacts/Ammo_Impact_Pistols_15}}
 {id=WavePlayer_52, properties {soundWaveAssetPtr=/Game/Sounds/Fort_Weapons/AmmoImpacts/Ammo_Impact_Pistols_14}}
 {id=WavePlayer_53, properties {soundWaveAssetPtr=/Game/Sounds/Fort_Weapons/AmmoImpacts/Ammo_Impact_Pistols_13}}
 {id=WavePlayer_54, properties {soundWaveAssetPtr=/Game/Sounds/Fort_Weapons/AmmoImpacts/Ammo_Impact_Pistols_12}}
 {id=WavePlayer_55, properties {soundWaveAssetPtr=/Game/Sounds/Fort_Weapons/AmmoImpacts/Ammo_Impact_Pistols_11}}
 {id=WavePlayer_56, properties {soundWaveAssetPtr=/Game/Sounds/Fort_Weapons/AmmoImpacts/Ammo_Impact_Pistols_10}}
 {id=WavePlayer_57, properties {soundWaveAssetPtr=/Game/Sounds/Fort_Weapons/AmmoImpacts/Ammo_Impact_Pistols_09}}
 {id=WavePlayer_58, properties {soundWaveAssetPtr=/Game/Sounds/Fort_Weapons/AmmoImpacts/Ammo_Impact_Pistols_08}}
 {id=WavePlayer_59, properties {soundWaveAssetPtr=/Game/Sounds/Fort_Weapons/AmmoImpacts/Ammo_Impact_Pistols_07}}
 {id=WavePlayer_60, properties {soundWaveAssetPtr=/Game/Sounds/Fort_Weapons/AmmoImpacts/Ammo_Impact_Pistols_06}}
 {id=WavePlayer_61, properties {soundWaveAssetPtr=/Game/Sounds/Fort_Weapons/AmmoImpacts/Ammo_Impact_Pistols_05}}
 {id=WavePlayer_62, properties {soundWaveAssetPtr=/Game/Sounds/Fort_Weapons/AmmoImpacts/Ammo_Impact_Pistols_04}}
 {id=WavePlayer_63, properties {soundWaveAssetPtr=/Game/Sounds/Fort_Weapons/AmmoImpacts/Ammo_Impact_Pistols_03}}
 {id=WavePlayer_64, properties {soundWaveAssetPtr=&Ammo_Impact_Pistols_02}}
 {id=WavePlayer_65, properties {soundWaveAssetPtr=&Ammo_Impact_Pistols_01}}
 {id=WavePlayer_66, properties {soundWaveAssetPtr=&Ammo_Impact_Sniper_01}}
 {id=WavePlayer_67, properties {soundWaveAssetPtr=&Ammo_Impact_Sniper_02}}
 {id=WavePlayer_68, properties {soundWaveAssetPtr=/Game/Sounds/Fort_Weapons/AmmoImpacts/Ammo_Impact_Sniper_03}}
 {id=WavePlayer_69, properties {soundWaveAssetPtr=/Game/Sounds/Fort_Weapons/AmmoImpacts/Ammo_Impact_Sniper_04}}
 {id=WavePlayer_70, properties {soundWaveAssetPtr=/Game/Sounds/Fort_Weapons/AmmoImpacts/Ammo_Impact_Sniper_05}}
 {id=WavePlayer_71, properties {soundWaveAssetPtr=/Game/Sounds/Fort_Weapons/AmmoImpacts/Ammo_Impact_Sniper_06}}
 {id=WavePlayer_72, properties {soundWaveAssetPtr=/Game/Sounds/Fort_Weapons/AmmoImpacts/Ammo_Impact_Sniper_07}}
 {id=WavePlayer_73, properties {soundWaveAssetPtr=/Game/Sounds/Fort_Weapons/AmmoImpacts/Ammo_Impact_Sniper_08}}
 {id=WavePlayer_74, properties {soundWaveAssetPtr=/Game/Sounds/Fort_Weapons/AmmoImpacts/Ammo_Impact_Sniper_09}}
 {id=WavePlayer_75, properties {soundWaveAssetPtr=/Game/Sounds/Fort_Weapons/AmmoImpacts/Ammo_Impact_Sniper_10}}
 {id=WavePlayer_76, properties {soundWaveAssetPtr=/Game/Sounds/Fort_Weapons/AmmoImpacts/Ammo_Impact_Sniper_11}}
 {id=WavePlayer_77, properties {soundWaveAssetPtr=/Game/Sounds/Fort_Weapons/AmmoImpacts/Ammo_Impact_Sniper_12}}
 {id=WavePlayer_78, properties {soundWaveAssetPtr=/Game/Sounds/Fort_Weapons/AmmoImpacts/Ammo_Impact_Sniper_13}}
 {id=WavePlayer_79, properties {soundWaveAssetPtr=/Game/Sounds/Fort_Weapons/AmmoImpacts/Ammo_Impact_Sniper_14}}
 {id=WavePlayer_80, properties {soundWaveAssetPtr=/Game/Sounds/Fort_Weapons/AmmoImpacts/Ammo_Impact_Sniper_15}}
 {id=WavePlayer_81, properties {soundWaveAssetPtr=/Game/Sounds/Fort_Weapons/AmmoImpacts/Ammo_Impact_Sniper_16}}
 {id=WavePlayer_82, properties {soundWaveAssetPtr=/Game/Sounds/Fort_Weapons/AmmoImpacts/Ammo_Impact_Sniper_17}}
 {id=WavePlayer_83, properties {soundWaveAssetPtr=/Game/Sounds/Fort_Weapons/AmmoImpacts/Ammo_Impact_Sniper_18}}
 {id=WavePlayer_84, properties {soundWaveAssetPtr=&Ammo_Impact_Shotgun_13}}
 {id=WavePlayer_85, properties {soundWaveAssetPtr=&Ammo_Impact_Shotgun_12}}
]
properties {
 attenuationSettings=/Game/Sounds/Attenuation/Fort_Player_Equip_Generic_Att
 bOverrideConcurrency=true
 maxDistance=1800
 soundClassObject=&Fort_WeaponRanged_Projectile_Effects
 volumeMultiplier=0.25
 concurrencyOverrides {maxCount=2}
}
