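# Niagara system that ejects weapon shell casings for looping fire, with user-driven type, velocity, size, spin, color and spawn probability.
# Shell_Eject_Loop_NoAudio bursts one casing, tagging it by Shell Type to pick among five mesh renderers.
# CasingSmoke_NoAudio reads those casings to spawn optional rising smoke ribbons, sized by type and faded by time of day.
# Despite the name, the audio emitter, with platform-gated collision sounds, and its smoke emitter are disabled.
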NiagaraSystem /Game/Weapons/FORT_Generic_Effects/ShellEject/NS_ShellsFire_Sound_Looping
dependencies {
 usedDataInterfaces [
  &NiagaraDataInterfaceAudioPlayer
  &NiagaraDataInterfacePlatformSet
  &NiagaraDataInterfaceCurve
  &NiagaraDataInterfaceParticleRead
  &NiagaraDataInterfaceArrayFloat
  &NiagaraDataInterfaceColorCurve
 ]
 usedDynamicInputs []
 usedModules [&SystemState]
}
emitters {
 CasingSmoke {
  bEnabled=false
  emitterUpdateScript [
   &EmitterState()
   &SpawnRate(
    SpawnRate=&MakeCustomFloatFromBool(Bool=$User.UseShellSmoke, True Float=60, False Float=0)
    Use Spawn Probability=false
   ) {enabled=false}
   &SpawnParticlesFromOtherEmitter(
    Attribute Reader=DataInterface() {propertyValues {EmitterBinding {bindingMode=Other, emitterName=Shell_Eject_Loop}}}
    Maximum Age Spawn Threshold=0.8
    Spawn Rate=&MakeCustomFloatFromBool(Bool=$User.UseShellSmoke, True Float=60, False Float=0)
    Use Maximum Age=true
    Use Minimum Age=false
   )
  ]
  particleSpawnScript [
   &InitializeRibbon(
    Color=<rgba 1 1 1 0.2>
    Lifetime=0.32
    Ribbon Width=&SelectFloatFromArray(Direct Array Index=$User.Shell Type, Float Selection Array=DataInterface() {propertyValues {FloatData [5,7,10,5,10]}})
    UsePositionOffset=true
    Write Material Random=true
   )
   &SampleParticlesFromOtherEmitter(
    Applied Velocity Scale=0
    Lifetime Sampling:ENiagara_DirectReadSamplingMode=::NewEnumerator1
    Particle ID Sampling=::NewEnumerator1
    Ribbon UV Distance:ENiagara_DirectReadSamplingMode=::NewEnumerator0
    Velocity Sampling=::NewEnumerator0
   )
   &AddVelocity(
    Coordinate Space=::World
    Velocity=&MakeVector(X=0, Y=0, Z=&Multiply_Float(A=50, B=1))
   ) {module=AddVelocity001}
   /SetVariables(Particles.Age_Sampled:NiagaraFloat=$Output.SampleParticlesFromOtherEmitter.SampledAge) {isSetParameters=true, module=SetVariables_#1}
   &DynamicMaterialParameters(
    Index 0 Param 1=1
    Index 0 Param 4=1
    Module.Param0WriteEnabled=true
    Module.Param1WriteEnabled=false
    Module.Param2WriteEnabled=false
    Module.Param3WriteEnabled=false
   ) {module=DynamicMaterialParameters001}
  ]
  particleUpdateScript [
   &ParticleState()
   &AccelerationForce(Acceleration=<0 0 150>)
   &SolveForcesAndVelocity(Clamp Velocity=false, Limit Acceleration=false)
   &ScaleRibbonWidth(Ribbon Width Scale=&FloatFromCurve(FloatCurve=DataInterface() {propertyValues {Curve {keys [<at 0 1>,<at 1 0>]}}}))
   &ScaleColor(
    Scale Alpha=&Multiply_Float(
     A=&Multiply_Float(
      A=&FloatFromCurve(
       FloatCurve=DataInterface() {propertyValues {Curve {keys [{arriveTangent=-1.7714, interpMode=RCIM_Cubic, leaveTangent=-1.7714, tangentMode=RCTM_User, time=0, value=1}, <at 1 0>]}}}
       Scale Curve=&FNTimeOfDayScalar(Night Scale=0.1, Morning Scale=0.3, Day Scale=0.7, Evening Scale=0.5)
      )
      B=&FloatFromCurve(
       FloatCurve=DataInterface() {propertyValues {Curve {keys [<at 0.1 0>,<at 0.25 1>,<at 0.6 1>,<at 1 0>]}}}
       CurveIndex=&Divide_Float(A=$Particles.Age_Sampled, B=0.8)
      )
     )
     B=&FloatFromCurve(FloatCurve=DataInterface() {propertyValues {Curve {keys [<at 0.3 1 RCIM_Cubic>,<at 1 0 RCIM_Cubic>]}}})
    )
    Scale RGB=&MakeVectorFromLinearColorRGB(Color=&ColorFromCurve(ColorCurve=DataInterface() {propertyValues {
        AlphaCurve {keys []}
        BlueCurve {keys [<at 0 1 RCIM_Cubic>,<at 1 1 RCIM_Cubic>]}
        GreenCurve {keys [<at 0 1 RCIM_Cubic>,<at 1 1 RCIM_Cubic>]}
        RedCurve {keys [<at 0 1 RCIM_Cubic>,<at 1 1 RCIM_Cubic>]}
       }}))
    ScaleA=true
    ScaleRGB=true
   )
  ]
  rendererClasses [&NiagaraRibbonRendererProperties]
  renderers [{rendererClass=&NiagaraRibbonRendererProperties}]
 }
 CasingSmoke_NoAudio {
  emitterUpdateScript [
   &EmitterState()
   &SpawnRate(
    SpawnRate=&MakeCustomFloatFromBool(Bool=$User.UseShellSmoke, True Float=60, False Float=0)
    Use Spawn Probability=false
   ) {enabled=false}
   &SpawnParticlesFromOtherEmitter(
    Attribute Reader=DataInterface() {propertyValues {EmitterBinding {bindingMode=Other, emitterName=Shell_Eject_Loop_NoAudio}}}
    Maximum Age Spawn Threshold=0.8
    Spawn Rate=&MakeCustomFloatFromBool(Bool=$User.UseShellSmoke, True Float=60, False Float=0)
    Use Maximum Age=true
    Use Minimum Age=false
   )
  ]
  particleSpawnScript [
   &InitializeRibbon(
    Color=<rgba 1 1 1 0.2>
    Lifetime=0.32
    Ribbon Width=&SelectFloatFromArray(Direct Array Index=$User.Shell Type, Float Selection Array=DataInterface() {propertyValues {FloatData [5,7,10,5,10]}})
    UsePositionOffset=true
    Write Material Random=true
   )
   &SampleParticlesFromOtherEmitter(
    Applied Velocity Scale=0
    Lifetime Sampling:ENiagara_DirectReadSamplingMode=::NewEnumerator1
    Particle ID Sampling=::NewEnumerator1
    Ribbon UV Distance:ENiagara_DirectReadSamplingMode=::NewEnumerator0
    Velocity Sampling=::NewEnumerator0
   )
   &AddVelocity(
    Coordinate Space=::World
    Velocity=&MakeVector(X=0, Y=0, Z=&Multiply_Float(A=50, B=1))
   ) {module=AddVelocity001}
   /SetVariables(Particles.Age_Sampled:NiagaraFloat=$Output.SampleParticlesFromOtherEmitter.SampledAge) {isSetParameters=true, module=SetVariables_#1}
   &DynamicMaterialParameters(
    Index 0 Param 1=1
    Index 0 Param 4=1
    Module.Param0WriteEnabled=true
    Module.Param1WriteEnabled=false
    Module.Param2WriteEnabled=false
    Module.Param3WriteEnabled=false
   ) {module=DynamicMaterialParameters001}
  ]
  particleUpdateScript [
   &ParticleState()
   &AccelerationForce(Acceleration=<0 0 150>)
   &SolveForcesAndVelocity(Clamp Velocity=false, Limit Acceleration=false)
   &ScaleRibbonWidth(Ribbon Width Scale=&FloatFromCurve(FloatCurve=DataInterface() {propertyValues {Curve {keys [<at 0 1>,<at 1 0>]}}}))
   &ScaleColor(
    Scale Alpha=&Multiply_Float(
     A=&Multiply_Float(
      A=&FloatFromCurve(
       FloatCurve=DataInterface() {propertyValues {Curve {keys [{arriveTangent=-1.7714, interpMode=RCIM_Cubic, leaveTangent=-1.7714, tangentMode=RCTM_User, time=0, value=1}, <at 1 0>]}}}
       Scale Curve=&FNTimeOfDayScalar(Night Scale=0.1, Morning Scale=0.3, Day Scale=0.7, Evening Scale=0.5)
      )
      B=&FloatFromCurve(
       FloatCurve=DataInterface() {propertyValues {Curve {keys [<at 0.1 0>,<at 0.25 1>,<at 0.6 1>,<at 1 0>]}}}
       CurveIndex=&Divide_Float(A=$Particles.Age_Sampled, B=0.8)
      )
     )
     B=&FloatFromCurve(FloatCurve=DataInterface() {propertyValues {Curve {keys [<at 0.3 1 RCIM_Cubic>,<at 1 0 RCIM_Cubic>]}}})
    )
    Scale RGB=&MakeVectorFromLinearColorRGB(Color=&ColorFromCurve(ColorCurve=DataInterface() {propertyValues {
        AlphaCurve {keys []}
        BlueCurve {keys [<at 0 1 RCIM_Cubic>,<at 1 1 RCIM_Cubic>]}
        GreenCurve {keys [<at 0 1 RCIM_Cubic>,<at 1 1 RCIM_Cubic>]}
        RedCurve {keys [<at 0 1 RCIM_Cubic>,<at 1 1 RCIM_Cubic>]}
       }}))
    ScaleA=true
    ScaleRGB=true
   )
  ]
  rendererClasses [&NiagaraRibbonRendererProperties]
  renderers [{rendererClass=&NiagaraRibbonRendererProperties}]
 }
 Shell_Eject_Loop {
  bEnabled=false
  emitterSpawnScript [
   &InitializeAudioPlayer(Audio Player=DataInterface() {propertyValues {
      Attenuation=null
      Concurrency=null
      MaxPlaysPerTick=10
      SoundToPlay=/Game/Sounds/Fort_Weapons/Shells/Weapon_Shells_Cue
      bAllowLoopingOneShotSounds=false
      bLimitPlaysPerTick=true
      bOnlyActiveDuringGameplay=false
      bStopWhenComponentIsDestroyed=true
      ConfigurationUserParameter {parameter {name=None, typeDefHandle {registeredTypeIndex=1}, varData []}}
      ParameterNames [ShellType,bEnableDelay]
     }})
   /SetVariables(Emitter.bPlayShellSounds:NiagaraBool=$User.IsSystemLocal) {isSetParameters=true}
   /SetVariables(Emitter.bEnableShellCollision:NiagaraBool=&IsPlatformSetActive(New Platform Set=DataInterface() {propertyValues {Platforms {qualityLevelMask=-24}}})) {isSetParameters=true}
   /Niagara/Modules/Audio/SetInitialAudioParameter(BoolValue:NiagaraBool=&Boolean_NotOperation(Bool=$Emitter.bEnableShellCollision), Parameter Name Index=1) {module=SetInitialAudioParameter001}
   &SetInitialAudioParameter(IntValue:NiagaraInt32=$User.Shell Type, Parameter Type:EInitialAudioParamType=::NewEnumerator1) {module=SetInitialAudioParameter002}
  ]
  emitterUpdateScript [
   &EmitterState()
   &SpawnBurst_Instantaneous(Spawn Count=1, Spawn Probability=$User.SpawnProbability, Use Spawn Probability=true)
  ]
  particleSpawnScript [
   &InitializeParticle(Mass Mode=::NewEnumerator1, Mesh Scale Mode=::NewEnumerator4, UsePositionOffset=false, Write Material Random=true)
   /SetVariables(Particles.VisibilityTag=$User.Shell Type) {isSetParameters=true, module=SetVariables_#2}
   /SetVariables(Particles.UniformMeshOrientation_Initial:Vector3f=&RandomRangeVector(Minimum=<0.2 0.2 0.2>, Maximum=<1 1 1>)) {isSetParameters=true, module=SetVariables_#3}
   &CylinderLocation(
    Cylinder Height:NiagaraFloat=$User.CylinderHeight
    Cylinder Radius=$User.CylinderRadius
    MaskSpawnGroup=false
    Module.Enable Lathe Profile=false
    Module.Override Randomness=false
    Module.Surface Only=false
    Orientation Axis:ENiagaraOrientationAxis=::XAxis
   )
   &AddVelocity(Velocity=$User.Velocity)
  ]
  particleUpdateScript [
   &ParticleState()
   &AccelerationForce(Acceleration=$User.Gravity)
   &InheritVelocity(Module.Limit Speed=true, Velocity Limit=10, Velocity Scale=$User.InheritParentVelocity)
   /SetVariables(Particles.CollisionChecks:NiagaraBool=&Boolean_AndOperation(
     Bool1=&SetBoolByFloatComparison(A=$Engine.Owner.LODDistance, B=1500, Comparison Type=::NewEnumerator6)
     Bool2=&SetBoolByFloatComparison(A=$Particles.Age, B=0.2)
    )) {isSetParameters=true, module=SetVariables_#4}
   &Collision(
    Age Colliding Particles=false
    CPU Collision Type=::NewEnumerator3
    Collision Enabled=&Boolean_AndOperation(
     Bool1=&Boolean_AndOperation(Bool1=$Emitter.bPlayShellSounds, Bool2=$Emitter.bEnableShellCollision)
     Bool2=$StackContext.CollisionChecks
    )
    Friction=0.7
    Module.Control Rest State Via Surface Normal=false
    Module.EnableMaxCollisionCount=false
    Module.Randomize Collision Normal Vector=true
    Radius Calculation Type=::NewEnumerator1
    Randomize Collision Normal=0.12
    Restitution=0.4
   )
   &SolveForcesAndVelocity(Clamp Velocity=false, Limit Acceleration=false)
   &ScaleMeshSize(Scale Factor=$User.Size)
   &ScaleMeshSize(Scale Factor=&VectorFromFloat(Value=&FloatFromCurve(FloatCurve=DataInterface() {propertyValues {Curve {keys [<at 0.7517 1 RCIM_Cubic>, {interpMode=RCIM_Cubic, tangentMode=RCTM_User, time=0.8642, value=1.3298}, <at 1 0 RCIM_Cubic>]}}}))) {module=ScaleMeshSize001}
   /Niagara/Modules/Update/Orientation/V2/UpdateMeshOrientation(Rotation Vector:Vector3f=&Multiply_Vector(A=$Particles.UniformMeshOrientation_Initial, B=$User.ShellsRotationRate)) {enabled=false}
   /Niagara/Modules/Update/Orientation/V3/UpdateMeshOrientation(Rotation Vector=&Multiply_Vector(A=$Particles.UniformMeshOrientation_Initial, B=$User.ShellsRotationRate)) {module=UpdateMeshOrientation001}
   &DynamicMaterialParameters(
    Index 0 Param 1=&MakeFloatFromLinearColor(LinearColor=$User.ShellColor, Channel:ENiagaraLinearColor_Channels=::NewEnumerator0)
    Index 0 Param 2=&MakeFloatFromLinearColor(LinearColor=$User.ShellColor, Channel:ENiagaraLinearColor_Channels=::NewEnumerator1)
    Index 0 Param 3=&MakeFloatFromLinearColor(LinearColor=$User.ShellColor, Channel:ENiagaraLinearColor_Channels=::NewEnumerator2)
    Module.Param0WriteEnabled=true
    Module.Param1WriteEnabled=true
    Module.Param2WriteEnabled=true
    Module.Param3WriteEnabled=false
   )
   /Niagara/Modules/Audio/PlayAudio(
    Audio Playback Position:NiagaraPosition=&MakeCustomVectorFromBool(
     Bool=$Emitter.bEnableShellCollision
     True Vector=Linked() {linkedVariable {name=Particles.Position, type=NiagaraPosition}}
     False Vector=&AddVectorToPosition(Position=$Particles.Position, Vector=<-30 -30 -100>)
    )
    Play Audio Trigger:NiagaraBool=&Boolean_OrOperation(
     Bool1=&Boolean_AndOperation(
      Bool1=$Particles.HasCollided
      Bool2=&SetBoolByIntComparison(A=$Particles.Number Of Collisions, B=1, Comparison Type=::NewEnumerator7)
     )
     Bool2=&Boolean_AndOperation(Bool1=&Boolean_NotOperation(Bool=$Emitter.bEnableShellCollision), Bool2=$Emitter.bPlayShellSounds)
    )
   )
  ]
  rendererClasses [&NiagaraMeshRendererProperties]
  renderers [
   {rendererClass=&NiagaraMeshRendererProperties, rendererIndex=0}
   {rendererClass=&NiagaraMeshRendererProperties, rendererIndex=1}
   {rendererClass=&NiagaraMeshRendererProperties, rendererIndex=2}
   {rendererClass=&NiagaraMeshRendererProperties, rendererIndex=3}
   {rendererClass=&NiagaraMeshRendererProperties, rendererIndex=4}
  ]
 }
 Shell_Eject_Loop_NoAudio {
  emitterUpdateScript [
   &EmitterState()
   &SpawnBurst_Instantaneous(Spawn Count=1, Spawn Probability=$User.SpawnProbability, Use Spawn Probability=true)
  ]
  particleSpawnScript [
   &InitializeParticle(Mass Mode=::NewEnumerator1, Mesh Scale Mode=::NewEnumerator4, UsePositionOffset=false, Write Material Random=true)
   /SetVariables(Particles.VisibilityTag=$User.Shell Type) {isSetParameters=true, module=SetVariables_#2}
   /SetVariables(Particles.UniformMeshOrientation_Initial:Vector3f=&RandomRangeVector(Minimum=<0.2 0.2 0.2>, Maximum=<1 1 1>)) {isSetParameters=true, module=SetVariables_#3}
   &CylinderLocation(
    Cylinder Height:NiagaraFloat=$User.CylinderHeight
    Cylinder Radius=$User.CylinderRadius
    MaskSpawnGroup=false
    Module.Enable Lathe Profile=false
    Module.Override Randomness=false
    Module.Surface Only=false
    Orientation Axis:ENiagaraOrientationAxis=::XAxis
   )
   &AddVelocity(Velocity=&Multiply_VectorByFloat(Vector=$User.Velocity, Float=&RandomRangeFloat(Minimum=0.75, Maximum=1)))
  ]
  particleUpdateScript [
   &ParticleState()
   &AccelerationForce(Acceleration=$User.Gravity)
   &InheritVelocity(Module.Limit Speed=true, Velocity Limit=10, Velocity Scale=$User.InheritParentVelocity)
   /SetVariables(Particles.CollisionChecks:NiagaraBool=&Boolean_AndOperation(
     Bool1=&SetBoolByFloatComparison(A=$Engine.Owner.LODDistance, B=1500, Comparison Type=::NewEnumerator6)
     Bool2=&SetBoolByFloatComparison(A=$Particles.Age, B=0.2)
    )) {isSetParameters=true, module=SetVariables_#4}
   &SolveForcesAndVelocity(Clamp Velocity=false, Limit Acceleration=false)
   &ScaleMeshSize(Scale Factor=$User.Size)
   &ScaleMeshSize(Scale Factor=&VectorFromFloat(Value=&FloatFromCurve(FloatCurve=DataInterface() {propertyValues {Curve {keys [<at 0.7517 1 RCIM_Cubic>, {interpMode=RCIM_Cubic, tangentMode=RCTM_User, time=0.8642, value=1.3298}, <at 1 0 RCIM_Cubic>]}}}))) {module=ScaleMeshSize001}
   &UpdateMeshOrientation(Rotation Vector=&Multiply_Vector(A=$Particles.UniformMeshOrientation_Initial, B=$User.ShellsRotationRate)) {module=UpdateMeshOrientation001}
   &DynamicMaterialParameters(
    Index 0 Param 1=&MakeFloatFromLinearColor(LinearColor=$User.ShellColor, Channel:ENiagaraLinearColor_Channels=::NewEnumerator0)
    Index 0 Param 2=&MakeFloatFromLinearColor(LinearColor=$User.ShellColor, Channel:ENiagaraLinearColor_Channels=::NewEnumerator1)
    Index 0 Param 3=&MakeFloatFromLinearColor(LinearColor=$User.ShellColor, Channel:ENiagaraLinearColor_Channels=::NewEnumerator2)
    Module.Param0WriteEnabled=true
    Module.Param1WriteEnabled=true
    Module.Param2WriteEnabled=true
    Module.Param3WriteEnabled=false
   )
  ]
  rendererClasses [&NiagaraMeshRendererProperties]
  renderers [
   {rendererClass=&NiagaraMeshRendererProperties, rendererIndex=0}
   {rendererClass=&NiagaraMeshRendererProperties, rendererIndex=1}
   {rendererClass=&NiagaraMeshRendererProperties, rendererIndex=2}
   {rendererClass=&NiagaraMeshRendererProperties, rendererIndex=3}
   {rendererClass=&NiagaraMeshRendererProperties, rendererIndex=4}
  ]
 }
}
summary {
 systemName=NS_ShellsFire_Sound_Looping
 emitters [
  {bEnabled=false, emitterName=Shell_Eject_Loop, rendererClasses [&NiagaraMeshRendererProperties]}
  {bEnabled=false, emitterName=CasingSmoke, rendererClasses [&NiagaraRibbonRendererProperties]}
  {emitterName=Shell_Eject_Loop_NoAudio, rendererClasses [&NiagaraMeshRendererProperties]}
  {emitterName=CasingSmoke_NoAudio, rendererClasses [&NiagaraRibbonRendererProperties]}
 ]
 userVariables [
  User.CylinderHeight {type=NiagaraFloat, defaultValue=1, description=""}
  User.CylinderRadius {type=NiagaraFloat, defaultValue=0, description=""}
  User.Gravity {type=Vector3f, description="", defaultValue {struct=Vector3f, value=<0 0 -1500>}}
  User.InheritParentVelocity {type=Vector3f, description="", defaultValue {struct=Vector3f, value=<0.2 0.2 0.2>}}
  User.IsSystemLocal {type=NiagaraBool}
  User.Shell Type {type=NiagaraInt32}
  User.ShellColor {type=LinearColor, description="", defaultValue {struct=LinearColor, value=<rgba 0 0 0 1>}}
  User.ShellsRotationRate {type=Vector3f, description="", defaultValue {struct=Vector3f, value=<1 1 1>}}
  User.Size {type=Vector3f, description="", defaultValue {struct=Vector3f, value=<1 1 1>}}
  User.SpawnProbability {type=NiagaraFloat, defaultValue=1, description=""}
  User.SpawnRateScale {type=NiagaraFloat, defaultValue=1, description=""}
  User.UseShellSmoke {type=NiagaraBool}
  User.Velocity {type=Vector3f, description="", defaultValue {struct=Vector3f, value=<0 0 0>}}
 ]
}
